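# Master parent material for RustedTrack set material instances, a Nanite-enabled opaque surface built largely on NeonCity functions.
# Four-layer textures with quadrant tints combine with switchable world-aligned base and masked top layers, simplified on lower quality and mobile; layered normal and specular blends sit unconnected.
# Optional fake windows add randomized interior cubemaps from a texture array, blinds and time-of-day lighting.
# Attributes pass through detail/SFX and water-wetness functions, with optional hit-effect WPO and debug previews.
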Material /Hera/Sets/RustedTrack/Materials/_Master/MM_RustedTrack_MIC_Parent
graph {
 properties {bUseMaterialAttributes=true, bUsedWithNanite=true, pixelDepthOffsetMode=PDOM_Legacy, usedShadingModels=Default Lit}
 nodes [
  n1 = SetMaterialAttributes!(
   MaterialAttributes
   Base Color=NamedRerouteUsage()
   Metallic=NamedRerouteUsage()
   Specular=NamedRerouteUsage()
   Roughness=NamedRerouteUsage()
   Emissive Color=NamedRerouteUsage()
   Normal=NamedRerouteUsage()
   World Position Offset=NamedRerouteUsage()
  ) {attributeSetTypes [
    #1
    #2
    #3
    #4
    #5
    #6
    #7
   ]}
  Composite!(MA In=@n1) {subgraphName=Debug}
  MaterialFunctionCall!(In=MaterialFunctionCall!(SFX V3=NamedRerouteUsage(), Mat In=@n1) {materialFunction=/Game/Global/Base4Layers_2022/MaterialFunctions/MF_DetailsWS}.Result) {materialFunction=/Game/Environments/Asteria/Sets/NeonCity/MaterialFunctions/MF_WaterWetness}
  n2 = Divide(
   A=Add(A=WorldPosition().XYZ, B=VectorParameter() {parameterName=RandomMaskWorldUVBias, defaultValue=<rgba 0.51 0.51 0.01 1>}.RGB)
   B=VectorParameter() {parameterName=RandomMaskWorldUVDivisor, defaultValue=<rgba 512 512 384 1>}.RGB
  )
  n3 = MaterialFunctionCall!(
   PrimaryTextureCoordinates=NamedRerouteDeclaration(None=MaterialFunctionCall!(UseLocalSpace, WorldPosition=@n2, NormalOverride) {materialFunction=/Game/Environments/Apollo/Sets/Skyfire/Materials/MaterialFunctions/MF_WorldAlignedUVs_HardEdge}.Result) {name=FakeWindowsUVs, nodeColor=<rgba 1 0 1 1>}
   2ndOrderTextureCoordinates
   RandomSeed
   NoiseDensityBias
   Bypass Noise Influence Crosstalk
  ) {materialFunction=/Game/Effects/Fort_Effects/Materials/_Functions/MF_RandomUVSample}
  MaterialExpressionAppend3Vector_2 = MaterialXAppend3Vector!(
   A=@n3.NoiseWithRange
   B=ConstantBiasScale(None=Sine(None=@n3.NoiseWithRange) {period=0.55})
   C=ConstantBiasScale(None=Cosine(None=@n3.NoiseWithRange) {period=1.95})
  )
  MaterialFunctionCall(Float3=NamedRerouteDeclaration(None=@MaterialExpressionAppend3Vector_2) {name=WindowRandomMask.RGB, nodeColor=<rgba 1 0 0.1406 1>}) {materialFunction=&BreakOutFloat3Components}
  n4 = NamedRerouteUsage()
  n5 = Multiply(A=ScalarParameter() {defaultValue=1, group="1) Base Layer", parameterName=BaseLayer_Roughness_WorldAligned_Strength}, B=NamedRerouteUsage())
  n6 = LinearInterpolate(
   A=@n5
   B=Multiply(A=ScalarParameter() {defaultValue=1, group="2) Top Layer", parameterName=TopLayer_Roughness_WorldAligned_Strength}, B=NamedRerouteUsage())
   Alpha=NamedRerouteUsage()
  )
  n7 = NamedRerouteUsage()
  n8 = Add(
   A=Multiply(A=@n4, B=StaticSwitchParameter(True=@n6, False=@n5) {defaultValue=true, group="2) Top Layer", parameterName=* UseTopLayer})
   B=@n7
  )
  n9 = ShadingPathSwitch(
   Default=QualitySwitch(
    Default=StaticSwitchParameter(True=@n8, False=@n7) {defaultValue=true, group="1) Base Layer", parameterName=* UseBaseLayer}
    Low=@n4
    High=@n4
    Medium=@n4
   )
   Mobile=@n4
  ) {desc=Migrated from FeatureLevelSwitch}
  MaterialProxyReplace(Realtime=@n9, MaterialProxy=@n4)
  n10 = Divide(A=NamedRerouteUsage(), B=ScalarParameter() {defaultValue=310, group="1) Base Layer", parameterName=BaseLayerUVsDivisor}) {constB=500}
  n11 = Divide(A=WorldPosition().XYZ, B=ScalarParameter() {defaultValue=560, group="1) Base Layer", parameterName=BaseLayerUVsCurlDivisor})
  n12 = ComponentMask(None=Multiply(
    A=VectorNoise(World Position=@n11) {noiseFunction=VNF_CurlALU}
    B=ScalarParameter() {defaultValue=0.1, group="1) Base Layer", parameterName=BaseLayerUVsCurlAmount}
   )) {g=true, r=true}
  NamedRerouteDeclaration(None=StaticSwitchParameter(True=Add(A=@n10, B=@n12), False=@n10) {group="1) Base Layer", parameterName=BaseLayerUVsUseCurlNoise}) {name=BaseLayerUVs, nodeColor=<rgba 0 0.4844 1 1>}
  n13 = WorldPosition()
  n14 = StaticSwitchParameter(True=Add(A=@n13.XYZ, B=VectorParameter() {group=World Aligned UVs, parameterName=WorldPositionBias}.RGB), False=@n13.XYZ) {group=World Aligned UVs, parameterName=BiasWorldUVs}
  n15 = Constant3Vector() {constant=<rgba 1.3333 1.3333 1 0>}
  n16 = MaterialFunctionCall(WorldPosition=StaticSwitchParameter(True=Divide(A=@n14, B=@n15), False=@n14) {group=World Aligned UVs, parameterName=ScaleWorldUVsToBuilderGrid}) {materialFunction=&MF_WorldAlignedUVs_HardEdge}
  NamedRerouteDeclaration(None=@n16.Result) {name=TriPlanarWorldUVs, nodeColor=<rgba 1 0.5625 0 1>}
  n17 = TextureSampleParameter2D(UVs=NamedRerouteUsage()) {
   group="Alpha Masks (Top Layer)"
   parameterName=AlphaMasksMap
   samplerSource=SSM_Wrap_WorldGroupSettings
   samplerType=SAMPLERTYPE_Masks
   texture=/Game/Global/Textures/Default/Blanks/T_White
  }
  n18 = ChannelMaskParameter(None=@n17.RGB) {group="Alpha Masks (Top Layer)", maskChannel=Blue, parameterName=AlphaMask_A_Channel, defaultValue=<rgba 0 0 1 0>}
  n19 = Multiply(A=@n18.R, B=ScalarParameter() {defaultValue=1, group="Alpha Masks (Top Layer)", parameterName=AlphaMask_Layer1_Strength})
  n20 = TextureSampleParameter2D(UVs=NamedRerouteUsage()) {
   group="Alpha Masks (Top Layer)"
   parameterName=AlphaMasksMap
   samplerSource=SSM_Wrap_WorldGroupSettings
   samplerType=SAMPLERTYPE_Masks
   texture=&T_White
  }
  n21 = ChannelMaskParameter(None=@n20.RGB) {group="Alpha Masks (Top Layer)", maskChannel=Green, parameterName=AlphaMask_B_Channel, defaultValue=<rgba 0 1 0 0>}
  n22 = Multiply(A=@n21.R, B=ScalarParameter() {defaultValue=1, group="Alpha Masks (Top Layer)", parameterName=AlphaMask_Layer2_Strength})
  NamedRerouteDeclaration(None=Power(Base=Max(A=@n19, B=@n22), Exp=ScalarParameter() {defaultValue=1, group="Alpha Masks (Top Layer)", parameterName=AlphaMasksExponent})) {name=AlphaMaskTopLayer, nodeColor=<rgba 0.5703 0 1 1>}
  n23 = NamedRerouteUsage()
  n24 = Multiply(A=ScalarParameter() {defaultValue=1, group="1) Base Layer", parameterName=BaseLayer_Metallic_WorldAligned_Strength}, B=NamedRerouteUsage())
  n25 = ScalarParameter() {defaultValue=1, group="2) Top Layer", parameterName=TopLayer_Metallic_WorldAligned_Strength}
  n26 = NamedRerouteUsage()
  n27 = StaticSwitchParameter(True=LinearInterpolate(A=@n24, B=Multiply(A=@n25, B=@n26), Alpha=NamedRerouteUsage()), False=@n24) {defaultValue=true, group="2) Top Layer", parameterName=* UseTopLayer}
  n28 = NamedRerouteUsage()
  n29 = StaticSwitchParameter(True=Add(A=Multiply(A=@n23, B=@n27), B=@n28), False=@n28) {defaultValue=true, group="1) Base Layer", parameterName=* UseBaseLayer}
  n30 = StaticSwitchParameter(True=@n28, False=@n23) {parameterName=UseBasicMetalMask}
  NamedRerouteDeclaration(None=ShadingPathSwitch(Default=QualitySwitch(Default=@n29, Low=@n30, High=@n30, Medium=@n30), Mobile=@n28) {desc=Migrated from FeatureLevelSwitch}) {name=Metallic, nodeColor=<rgba 0 1 0.2969 1>}
  n31 = NamedRerouteUsage()
  NamedRerouteDeclaration(None=@n31) {name=Specular, nodeColor=<rgba 1 0.7031 0 1>}
  NamedRerouteDeclaration(None=@n9) {name=Roughness, nodeColor=<rgba 0.3125 1 0 1>}
  n32 = TextureSampleParameter2D(UVs=NamedRerouteUsage()) {
   group="1) Base Layer"
   parameterName=BaseLayerMasks
   samplerSource=SSM_Wrap_WorldGroupSettings
   samplerType=SAMPLERTYPE_Masks
   texture=&T_WhiteMask
  }
  n33 = MaterialFunctionCall(Float4=Subtract(A=@n32.RGBA) {constB=0.5}) {materialFunction=&BreakOutFloat4Components}
  NamedRerouteDeclaration(None=@n33.R) {name=RoughnessMap_BaseLayer_WorldAligned, nodeColor=<rgba 0.2422 1 0 1>}
  NamedRerouteDeclaration(None=Divide(A=NamedRerouteUsage(), B=ScalarParameter() {defaultValue=1500, group="Alpha Masks (Top Layer)", parameterName=AlphaMaskA_UVsDivisor}) {constB=500}) {name=AlphaMapAUVs, nodeColor=<rgba 0.1484 0 1 1>}
  n34 = TextureCoordinate()
  n35 = AppendVector(
   A=ScalarParameter() {defaultValue=2, group=Fake Windows, parameterName=BlindsScaleU}
   B=ScalarParameter() {defaultValue=1, group=Fake Windows, parameterName=BlindsScaleV}
  )
  n36 = Constant()
  n37 = NamedRerouteUsage()
  n38 = TextureSample(UVs=Add(A=Multiply(A=@n34, B=@n35), B=AppendVector(A=@n36, B=@n37))) {samplerType=SAMPLERTYPE_Masks, texture=/Game/Environments/Asteria/Sets/NeonCity/Techart/Textures/T_BlindsMask}
  n39 = SphereMask(
   A=@n38.R
   B=Constant()
   Radius=NamedRerouteUsage()
   Hardness=ScalarParameter() {defaultValue=0.5, group=Fake Windows, parameterName=BlindsEdgeHardness}
  )
  NamedRerouteDeclaration(None=OneMinus(None=Subtract(A=@n39, B=@n38.G))) {name=BlindsMask, nodeColor=<rgba 0.0312 1 0 1>}
  n40 = NamedRerouteUsage()
  NamedRerouteDeclaration(None=@n40) {name=Normals, nodeColor=<rgba 1 0 0.8906 1>}
  n41 = TextureSampleParameter2D(UVs=NamedRerouteUsage()) {
   group="2) Top Layer"
   parameterName=TopLayerMasks
   samplerSource=SSM_Wrap_WorldGroupSettings
   samplerType=SAMPLERTYPE_Masks
   texture=&T_WhiteMask
  }
  n42 = MaterialFunctionCall(Float4=Subtract(A=@n41.RGBA) {constB=0.5}) {materialFunction=&BreakOutFloat4Components}
  NamedRerouteDeclaration(None=@n42.R) {name=RoughnessMap_TopLayer_WorldAligned, nodeColor=<rgba 0 1 0.1562 1>}
  n43 = ScalarParameter() {defaultValue=80, group=Ambient Occlusion, parameterName=BaseDistance}
  n44 = ScalarParameter() {defaultValue=25, group=Ambient Occlusion, parameterName=Radius}
  n45 = OneMinus(None=NamedRerouteDeclaration(None=DistanceFieldApproxAO!(World Position, Normal, BaseDistance=@n43, Radius=@n44)) {name=DistanceFieldAO, nodeColor=<rgba 0 1 0.4844 1>})
  n46 = Add(A=@n45, B=ScalarParameter() {defaultValue=-0.1, group=Ambient Occlusion, parameterName=AOEdgeBias})
  n47 = ScalarParameter() {defaultValue=1, group=Ambient Occlusion, parameterName=AOEdgeScale}
  n48 = NamedRerouteUsage()
  n49 = Saturate(None=Add(A=Multiply(A=@n46, B=@n47) {constB=2}, B=Subtract(A=@n45, B=@n48)))
  n50 = ScalarParameter() {defaultValue=0.8, group=Ambient Occlusion, parameterName=AOStainStrength}
  n51 = NamedRerouteUsage()
  n52 = StaticSwitchParameter(True=LinearInterpolate(A=@n51, B=NamedRerouteUsage(), Alpha=NamedRerouteUsage()), False=@n51) {defaultValue=true, group="2) Top Layer", parameterName=* UseTopLayer}
  n53 = Constant() {r=1}
  n54 = Multiply(
   A=OneMinus(None=Multiply(A=@n49, B=@n50))
   B=StaticSwitchParameter(
    True=Constant() {r=1}
    False=StaticSwitchParameter(True=@n52, False=@n53) {defaultValue=true, group="1) Base Layer", parameterName=* UseBaseLayer}
   ) {group=Fake Windows, parameterName=UseFakeWindows}
  )
  n55 = Constant() {r=1}
  NamedRerouteDeclaration(None=ShadingPathSwitch(Default=QualitySwitch(Default=@n54, Low=@n55, High=@n55, Medium=@n55), Mobile=@n55) {desc=Migrated from FeatureLevelSwitch}) {name=AmbientOcclusion, nodeColor=<rgba 0.9922 1 0 1>}
  NamedRerouteDeclaration(None=@n33.G) {name=MetallicMap_BaseLayer_WorldAligned, nodeColor=<rgba 0 1 0 1>}
  NamedRerouteDeclaration(None=@n33.B) {name=AOMap_BaseLayer_WorldAligned, nodeColor=<rgba 0 0 1 1>}
  NamedRerouteDeclaration(None=@n42.G) {name=MetallicMap_TopLayer_WorldAligned, nodeColor=<rgba 0 1 0 1>}
  NamedRerouteDeclaration(None=@n42.B) {name=AOMap_TopLayer_WorldAligned, nodeColor=<rgba 0 0 1 1>}
  n56 = NamedRerouteUsage()
  n57 = NamedRerouteUsage()
  n58 = TextureSampleParameter2D(UVs=NamedRerouteUsage()) {
   group="1) Base Layer"
   parameterName=BaseLayerAlbedo
   samplerSource=SSM_Wrap_WorldGroupSettings
   texture=&T_WhitesRGB
  }
  n59 = Multiply(A=VectorParameter() {group="1) Base Layer", parameterName=BaseLayer_ColorScale, defaultValue=<rgba 1 1 1 0>}.RGB, B=@n58.RGB)
  n60 = TextureSampleParameter2D(UVs=NamedRerouteUsage()) {
   group="2) Top Layer"
   parameterName=TopLayerAlbedo
   samplerSource=SSM_Wrap_WorldGroupSettings
   texture=&T_WhitesRGB
  }
  n61 = Multiply(A=@n60.RGB, B=VectorParameter() {group="2) Top Layer", parameterName=TopLayer_ColorScale, defaultValue=<rgba 1 1 1 0>}.RGB)
  n62 = Add(A=NamedRerouteUsage(), B=ScalarParameter() {group="2) Top Layer", parameterName=TopLayerMaskBias})
  n63 = NamedRerouteDeclaration(None=Multiply(A=Saturate(None=@n62), B=ScalarParameter() {defaultValue=1, group="2) Top Layer", parameterName=TopLayerMaskScale})) {name=TopLayerMask, nodeColor=<rgba 0.6641 0 1 1>}
  n64 = LinearInterpolate(B=@n61, Alpha=@n63) {constA=0.5, constB=0.5}
  n65 = StaticSwitchParameter(
   True=StaticSwitchParameter(
    True=MaterialFunctionCall(Base=@n59, Blend=@n64) {materialFunction=&Blend_Overlay}.Result
    False=LinearInterpolate(A=@n59, B=@n61, Alpha=@n63)
   ) {group="2) Top Layer", parameterName=* OverlayTopLayer}
   False=@n59
  ) {defaultValue=true, group="2) Top Layer", parameterName=* UseTopLayer}
  n66 = StaticSwitchParameter(
   True=StaticSwitchParameter(
    True=MaterialFunctionCall(Base=@n57, Blend=@n65) {materialFunction=&Blend_Overlay}.Result
    False=@n65
   ) {defaultValue=true, group=Texture Data Asset, parameterName=Use4Layers}
   False=NamedRerouteUsage()
  ) {defaultValue=true, group="1) Base Layer", parameterName=* UseBaseLayer}
  n67 = NamedRerouteUsage()
  n68 = QualitySwitch(
   Default=StaticSwitchParameter(True=LinearInterpolate(A=@n56, B=@n66, Alpha=@n67), False=@n66) {defaultValue=true, group=Texture Data Asset, parameterName=Use4Layers}
   Low=@n56
   Medium=@n56
  )
  n69 = StaticSwitchParameter(True=LinearInterpolate(A=NamedRerouteUsage(), B=@n68, Alpha=NamedRerouteUsage()) {constB=0}, False=@n68) {group=Fake Windows, parameterName=UseFakeWindows}
  NamedRerouteDeclaration(None=ShadingPathSwitch(Default=@n69, Mobile=MaterialProxyReplace(Realtime=@n69, MaterialProxy=@n68)) {desc=Migrated from FeatureLevelSwitch}) {name=Base Color, nodeColor=<rgba 0.1016 0 1 1>}
  NamedRerouteDeclaration(None=Divide(A=NamedRerouteUsage(), B=ScalarParameter() {defaultValue=260, group="2) Top Layer", parameterName=TopLayerUVsDivisor}) {constB=500}) {name=TopLayerUVs, nodeColor=<rgba 1 0.75 0 1>}
  NamedRerouteDeclaration(None=Divide(A=NamedRerouteUsage(), B=ScalarParameter() {defaultValue=1800, group="Alpha Masks (Top Layer)", parameterName=AlphaMaskB_UVsDivisor}) {constB=500}) {name=AlphaMapBUVs, nodeColor=<rgba 0.1484 0 1 1>}
  n70 = TextureSampleParameter2D() {
   group=Texture Data Asset
   parameterName=SpecularMasks
   samplerSource=SSM_Wrap_WorldGroupSettings
   texture=/Game/Global/Textures/Default/Blanks/T_MidgreysRGB
  }
  n71 = TextureSampleParameter2D() {
   group=Texture Data Asset
   parameterName=SpecularMasks_2
   samplerSource=SSM_Wrap_WorldGroupSettings
   texture=&T_MidgreysRGB
  }
  n72 = TextureSampleParameter2D() {
   group=Texture Data Asset
   parameterName=SpecularMasks_3
   samplerSource=SSM_Wrap_WorldGroupSettings
   texture=&T_MidgreysRGB
  }
  n73 = TextureSampleParameter2D() {
   group=Texture Data Asset
   parameterName=SpecularMasks_4
   samplerSource=SSM_Wrap_WorldGroupSettings
   texture=&T_MidgreysRGB
  }
  n74 = MaterialFunctionCall(Float4=MaterialFunctionCall!(1=@n70.RGBA, 2=@n71.RGBA, 3=@n72.RGBA, 4=@n73.RGBA, UVOverride) {materialFunction=/Game/Environments/Asteria/Sets/NeonCity/Techart/MF_NeonCity_4Layer_Vec4}.Result) {materialFunction=&BreakOutFloat4Components}
  NamedRerouteDeclaration(None=@n74.R) {name=Specular Mask 4Layer, nodeColor=<rgba 1 0.6328 0 1>}
  n75 = StaticSwitchParameter(
   True=LinearInterpolate(
    A=@n74.G
    B=ScalarParameter() {defaultValue=0.5, group=Texture Data Asset, parameterName=FlattenMetallic_4Layer_Value}
    Alpha=ScalarParameter() {defaultValue=1, group=Texture Data Asset, parameterName=FlattenMetallic_4Layer_Amount}
   )
   False=@n74.G
  ) {group=Texture Data Asset, parameterName=FlattenMetallic_4Layer}
  n76 = NamedRerouteUsage()
  NamedRerouteDeclaration(None=StaticSwitchParameter(True=Multiply(A=@n75, B=@n76), False=@n75) {parameterName=UseMetallicModulate}) {name=Metallic Mask 4Layer, nodeColor=<rgba 0 1 0 1>}
  n77 = StaticSwitchParameter(
   True=LinearInterpolate(
    A=@n74.B
    B=ScalarParameter() {defaultValue=0.75, group=Texture Data Asset, parameterName=FlattenRoughness_4Layer_Value}
    Alpha=ScalarParameter() {defaultValue=0.75, group=Texture Data Asset, parameterName=FlattenRoughness_4Layer_Amount}
   )
   False=@n74.B
  ) {group=Texture Data Asset, parameterName=FlattenRoughness_4Layer}
  n78 = NamedRerouteUsage()
  NamedRerouteDeclaration(None=StaticSwitchParameter(True=Multiply(A=@n77, B=@n78), False=@n77) {parameterName=UseRoughnessModulate}) {name=Roughness Mask 4Layer, nodeColor=<rgba 0 0 1 1>}
  n79 = ScalarParameter() {defaultValue=1, group=Color_MatLayer1, parameterName=Metallic_MatLayer1_A}
  n80 = LinearInterpolate(A=@n79, B=ScalarParameter() {defaultValue=1, group=Color_MatLayer1, parameterName=Metallic_MatLayer1_B}, Alpha=NamedRerouteUsage())
  n81 = LinearInterpolate(
   A=ScalarParameter() {defaultValue=1, group=Color_MatLayer1, parameterName=Metallic_MatLayer1_C}
   B=ScalarParameter() {defaultValue=1, group=Color_MatLayer1, parameterName=Metallic_MatLayer1_D}
   Alpha=NamedRerouteUsage()
  )
  n82 = NamedRerouteUsage()
  n83 = Constant3Vector()
  n84 = AppendVector(A=StaticSwitchParameter(True=LinearInterpolate(A=@n80, B=@n81, Alpha=@n82), False=@n79) {parameterName=UseMetallicByQuadrants}, B=@n83)
  n85 = ScalarParameter() {defaultValue=1, group=Color_MatLayer2, parameterName=Metallic_MatLayer2_A}
  n86 = LinearInterpolate(A=@n85, B=ScalarParameter() {defaultValue=1, group=Color_MatLayer2, parameterName=Metallic_MatLayer2_B}, Alpha=NamedRerouteUsage())
  n87 = LinearInterpolate(
   A=ScalarParameter() {defaultValue=1, group=Color_MatLayer2, parameterName=Metallic_MatLayer2_C}
   B=ScalarParameter() {defaultValue=1, group=Color_MatLayer2, parameterName=Metallic_MatLayer2_D}
   Alpha=NamedRerouteUsage()
  )
  n88 = NamedRerouteUsage()
  n89 = AppendVector(A=StaticSwitchParameter(True=LinearInterpolate(A=@n86, B=@n87, Alpha=@n88), False=@n85) {parameterName=UseMetallicByQuadrants}, B=@n83)
  n90 = Constant4Vector()
  NamedRerouteDeclaration(None=ComponentMask(None=MaterialFunctionCall(1=@n84, 2=@n89, 3=@n90.RGBA, 4=@n90.RGBA) {materialFunction=&MF_NeonCity_4Layer_Vec4}.Result) {r=true}) {name=MetallicByQuad, nodeColor=<rgba 0 1 0.5312 1>}
  n91 = TextureSampleParameter2D() {
   group=Texture Data Asset
   parameterName=Normals
   samplerSource=SSM_Wrap_WorldGroupSettings
   samplerType=SAMPLERTYPE_Normal
   texture=&FlatNormalBlue
  }
  n92 = TextureSampleParameter2D() {
   group=Texture Data Asset
   parameterName=Normals_Texture_2
   samplerSource=SSM_Wrap_WorldGroupSettings
   samplerType=SAMPLERTYPE_Normal
   texture=&FlatNormalBlue
  }
  n93 = TextureSampleParameter2D() {
   group=Texture Data Asset
   parameterName=Normals_Texture_3
   samplerSource=SSM_Wrap_WorldGroupSettings
   samplerType=SAMPLERTYPE_Normal
   texture=&FlatNormalBlue
  }
  n94 = TextureSampleParameter2D() {
   group=Texture Data Asset
   parameterName=Normals_Texture_4
   samplerSource=SSM_Wrap_WorldGroupSettings
   samplerType=SAMPLERTYPE_Normal
   texture=&FlatNormalBlue
  }
  NamedRerouteDeclaration(None=MaterialFunctionCall(1=@n91.RGBA, 2=@n92.RGBA, 3=@n93.RGBA, 4=@n94.RGBA) {materialFunction=&MF_NeonCity_4Layer_Vec4}.Result) {name=Normals_4Layer, nodeColor=<rgba 0 0.2734 1 1>}
  n95 = TextureCoordinate()
  n96 = ScalarParameter() {defaultValue=0.005, parameterName=DistortionNoiseSize}
  n97 = Divide(A=PixelDepth().R, B=ScalarParameter() {defaultValue=4000, parameterName=DistortionNoiseDistance})
  n98 = Multiply(
   A=Sine(None=Divide(A=@n95, B=@n96))
   B=LinearInterpolate(
    A=ScalarParameter() {defaultValue=0.01, parameterName=DistortCubemapIntensity_Near}
    B=ScalarParameter() {defaultValue=0.005, parameterName=DistortCubemapIntensity_Far}
    Alpha=Saturate(None=@n97)
   )
  )
  n99 = TextureCoordinate()
  n100 = Multiply(
   A=ComponentMask(None=NamedRerouteUsage()) {g=true, r=true}
   B=ScalarParameter() {defaultValue=0.47, group=Fake Windows, parameterName=WindowRefractionScale}
  )
  n101 = WorldPosition()
  n102 = VectorParameter() {parameterName=RandomMaskWorldUVDivisor, defaultValue=<rgba 512 512 384 1>}
  n103 = VectorNoise(World Position=Add(A=Divide(A=@n101.XYZ, B=@n102.RGB), B=VectorParameter() {parameterName=RandomMaskWorldUVBias, defaultValue=<rgba 0.51 0.51 0.01 1>}.RGB) {constB=0.5})
  n104 = NamedRerouteDeclaration(None=ComponentMask(None=@n103) {r=true}) {name=WindowRandomMask.R, nodeColor=<rgba 1 0 0.8906 1>}
  n105 = AppendVector(A=@n104, B=NamedRerouteDeclaration(None=ComponentMask(None=@n103) {g=true}) {name=WindowRandomMask.G, nodeColor=<rgba 0 1 0 1>})
  n106 = ScalarParameter() {defaultValue=2.33, parameterName=RandomMult}
  n107 = MaterialFunctionCall!(
   UVs=NamedRerouteDeclaration(None=Add(A=@n99, B=@n100)) {name=InteriorCubeMapUV, nodeColor=<rgba 1 0 0.2344 1>}
   Custom Random Input=Ceil(None=Multiply(A=@n105, B=@n106))
   Tiling=Add(A=ScalarParameter() {bUseCustomPrimitiveData=true, parameterName=InteriorCubeMapScale})
  ) {materialFunction=/Game/Environments/Asteria/Sets/NeonCity/Techart/MF_NeonCity_InteriorCubemap}
  n108 = StaticSwitch(
   True=@n107.UVW Randomized
   False=@n107.UVW Randomized
   Value=StaticBoolParameter() {defaultValue=true, group=Fake Windows, parameterName=RandomizeCubemapRotation}
  )
  n109 = MaterialFunctionCall!(Vector=NamedRerouteDeclaration(None=QualitySwitch(Default=@n107.UVW, Epic=@n108)) {name=CubemapUVW, nodeColor=<rgba 1 0 0.5156 1>}) {materialFunction=/Engine/Functions/Engine_MaterialFunctions01/Coordinates/LongLatToUV}
  n110 = Multiply(A=ScalarParameter() {defaultValue=3, parameterName=TextureArrayLength}, B=NamedRerouteUsage())
  n111 = TextureSampleParameter2DArray!(UVs=AppendVector(A=Add(A=@n98, B=@n109.Result), B=Round(None=@n110)), Apply View MipBias) {parameterName=Window_Diffuse_LongLat, texture=/Game/Environments/Asteria/Sets/NeonCity/Techart/Textures/TA_FakeInterior_LongLat}
  n112 = Desaturation(None=@n111.RGB)
  n113 = ScalarParameter() {defaultValue=0.021, parameterName=DayLightCubeBrightness}
  n114 = Add(A=@n104, B=NamedRerouteDeclaration(None=ComponentMask(None=@n103) {b=true}) {name=WindowRandomMask.B, nodeColor=<rgba 0 0 1 1>})
  n115 = ScalarParameter() {defaultValue=0.5, group=Fake Windows, parameterName=CubemapContrast}
  n116 = Multiply(
   A=LinearInterpolate(
    A=VectorParameter() {group=Fake Windows, parameterName=CubemapTint1, defaultValue=<rgba 2 1.5334 1.0521 1>}.RGB
    B=VectorParameter() {group=Fake Windows, parameterName=CubemapTint2, defaultValue=<rgba 0.3229 0.7422 2 1>}.RGB
    Alpha=Frac(None=@n114)
   )
   B=Multiply(A=Power(Base=@n111.RGB, Exp=@n115), B=ScalarParameter() {defaultValue=0.5, group=Fake Windows, parameterName=UnoccupiedRoomScale})
  )
  n117 = Multiply(A=@n116, B=NamedRerouteUsage())
  n118 = Constant2Vector() {g=0.5, r=0.5}
  n119 = NamedRerouteUsage()
  n120 = NamedRerouteUsage()
  n121 = AppendVector(A=Subtract(A=AppendVector(A=@n119, B=@n120)) {constB=0.5}, B=Constant())
  n122 = Constant3Vector() {constant=<rgba 1 1 0 0>}
  n123 = Multiply(
   A=ComponentMask(None=MaterialFunctionCall(Vector=@n121, Default=@n122) {materialFunction=&SafeNormalize}.Result) {g=true, r=true}
   B=ScalarParameter() {defaultValue=2, parameterName=Mobile Window GradientRandom OffsetVector}
  )
  n124 = Divide(
   A=Distance(A=TextureCoordinate(), B=Add(A=@n118, B=@n123))
   B=ScalarParameter() {defaultValue=3.01, parameterName=Mobile Window Gradient_Radius}
  )
  n125 = ScalarParameter() {defaultValue=7.36, parameterName=Mobile Window Gradient_Falloff}
  n126 = VectorParameter() {parameterName=MobileWindowEmissive, defaultValue=<rgba 0.0614 0.1547 0.2 1>}
  n127 = Saturate(None=LinearInterpolate(
    A=ScalarParameter() {defaultValue=0.2, parameterName=Mobile Window Lit_Low}
    B=ScalarParameter() {defaultValue=1, parameterName=Mobile Window Lit_High}
    Alpha=NamedRerouteUsage()
   ))
  n128 = ShadingPathSwitch(
   Default=LinearInterpolate(
    A=Multiply(A=@n112, B=@n113)
    B=StaticSwitchParameter(True=@n117, False=@n116) {defaultValue=true, group=Fake Windows, parameterName=UseBlinds}
    Alpha=NamedRerouteUsage()
   )
   Mobile=Multiply(A=Power(Base=@n124, Exp=@n125), B=Multiply(A=@n126.RGB, B=@n127))
  ) {desc=Migrated from FeatureLevelSwitch}
  NamedRerouteDeclaration(None=@n128) {name=WindowDiffuse, nodeColor=<rgba 1 0.5625 0 1>}
  NamedRerouteDeclaration(None=@n128) {name=WindowEmissive, nodeColor=<rgba 0.2422 1 0 1>}
  n129 = ScalarParameter() {defaultValue=-0.15, group=Fake Windows, parameterName=LightsActive_Morning/Evening_Bias}
  n130 = MaterialFunctionCall(
   Value_Morning=@n129
   Value_Day=ScalarParameter() {defaultValue=-1, group=Fake Windows, parameterName=LightsActive_Day_Bias}
   Value_Evening=@n129
   Value_Night=ScalarParameter() {defaultValue=0.4, group=Fake Windows, parameterName=LightsActive_Night_Bias}
  ) {materialFunction=&Scalar_Time_of_Day}
  n131 = ScalarParameter() {defaultValue=0.4, group=Fake Windows, parameterName=LightsActive_Bias}
  n132 = Saturate(None=Add(
    A=NamedRerouteUsage()
    B=StaticSwitchParameter(True=@n130.Result, False=@n131) {defaultValue=true, group=Fake Windows, parameterName=LightsActive_UseTODBias}
   ))
  n133 = LinearInterpolate(
   A=ScalarParameter() {defaultValue=-0.5, group=Fake Windows, parameterName=WinLit_Min}
   B=ScalarParameter() {defaultValue=1, group=Fake Windows, parameterName=WinLit_Max}
   Alpha=NamedRerouteUsage()
  )
  NamedRerouteDeclaration(None=Multiply(A=Round(None=@n132), B=Saturate(None=@n133))) {name=Window_LightsActiveMask, nodeColor=<rgba 0 1 0.5078 1>}
  n134 = Constant() {desc=Future Emissive Goes Here}
  n135 = Add(A=NamedRerouteUsage(), B=@n134)
  n136 = NamedRerouteUsage()
  NamedRerouteDeclaration(None=StaticSwitchParameter(True=LinearInterpolate(A=@n135, B=@n134, Alpha=@n136) {constB=0}, False=@n134) {group=Fake Windows, parameterName=UseFakeWindows}) {name=Emissive, nodeColor=<rgba 0 1 0.3203 1>}
  NamedRerouteDeclaration(None=OneMinus(None=NamedRerouteDeclaration(None=@n74.A) {name=Alpha Mask Texture 4 Layer, nodeColor=<rgba 1 1 1 1>})) {name=Alpha Mask Texture 4 Layer Inverted, nodeColor=<rgba 1 1 1 1>}
  n137 = TextureObjectParameter() {
   desc="SFX TextureGrayscale You can reuse that texture across layers and functions. "
   group=0_SFX Global
   parameterName=SFX_Grayscale
   samplerType=SAMPLERTYPE_LinearGrayscale
   sortPriority=0
   texture=/Game/Global/Base4Layers_2022/Textures/T_SFX_Grayscale
  }
  n138 = TextureObjectParameter() {
   desc="SFX UniversalTextureMask Uses Dot to select channel used for various parts of material.You can reuse that texture across layers and functions. "
   group=0_SFX Global
   parameterName=SFX_Mask
   samplerType=SAMPLERTYPE_Masks
   sortPriority=0
   texture=/Game/Global/Base4Layers_2022/Textures/T_SFX_M
  }
  NamedRerouteDeclaration(None=MaterialFunctionCall!(In=StaticSwitchParameter(True=@n137, False=@n138) {desc=Make sure size is less than 512 and lower, group=0_SFX Global, parameterName=UseSFXGrayscale}) {materialFunction=/Game/Global/Base4Layers_2022/MaterialFunctions/MF_SFXTexture}.SFX) {name=SFX V3, nodeColor=<rgba 0.6406 0 1 1>}
  NamedRerouteDeclaration(None=StaticSwitchParameter(True=MaterialFunctionCall() {materialFunction=&MF_FN_DestructibleDamage}.DamageAnimWPO, False=Constant()) {group=World Position Offset, parameterName=UseHitEffects}) {name=WPO, nodeColor=<rgba 0 0.8359 1 1>}
  n139 = TextureSampleParameter2D() {
   group=Texture Data Asset
   parameterName=Diffuse
   samplerSource=SSM_Wrap_WorldGroupSettings
   texture=&T_MidgreysRGB
  }
  n140 = TextureSampleParameter2D() {
   group=Texture Data Asset
   parameterName=Diffuse_Texture_2
   samplerSource=SSM_Wrap_WorldGroupSettings
   texture=&T_MidgreysRGB
  }
  n141 = TextureSampleParameter2D() {
   group=Texture Data Asset
   parameterName=Diffuse_Texture_3
   samplerSource=SSM_Wrap_WorldGroupSettings
   texture=&T_MidgreysRGB
  }
  n142 = TextureSampleParameter2D() {
   group=Texture Data Asset
   parameterName=Diffuse_Texture_4
   samplerSource=SSM_Wrap_WorldGroupSettings
   texture=&T_MidgreysRGB
  }
  n143 = MaterialFunctionCall(1=@n139.RGBA, 2=@n140.RGBA, 3=@n141.RGBA, 4=@n142.RGBA) {materialFunction=&MF_NeonCity_4Layer_Vec4}
  n144 = VectorParameter() {group=Texture Data Asset, parameterName=FlattenDiffuse_4Layer_Color, defaultValue=<rgba 0.5 0.5 0.5 1>}
  n145 = LinearInterpolate(
   A=@n143.Result
   B=AppendVector(A=@n144.RGB, B=@n144.RGB)
   Alpha=ScalarParameter() {defaultValue=1, group=Texture Data Asset, parameterName=FlattenDiffuse_4Layer_Color_Lerp}
  )
  n146 = ComponentMask(None=NamedRerouteDeclaration(None=StaticSwitchParameter(True=@n145, False=@n143.Result) {group=Texture Data Asset, parameterName=FlattenDiffuse_4Layer}) {name=Diffuse_4Layer.RGBA, nodeColor=<rgba 0.1484 1 0 1>}) {b=true, g=true, r=true}
  n147 = VectorParameter() {group=Color_MatLayer1, parameterName=Color_MatLayer1_A, defaultValue=<rgba 1 1 1 1>}
  n148 = VectorParameter() {group=Color_MatLayer1, parameterName=Color_MatLayer1_B, defaultValue=<rgba 1 1 1 1>}
  n149 = NamedRerouteUsage()
  n150 = VectorParameter() {group=Color_MatLayer1, parameterName=Color_MatLayer1_C, defaultValue=<rgba 1 1 1 1>}
  n151 = VectorParameter() {group=Color_MatLayer1, parameterName=Color_MatLayer1_D, defaultValue=<rgba 1 1 1 1>}
  n152 = NamedRerouteUsage()
  n153 = StaticSwitchParameter(
   True=LinearInterpolate(
    A=LinearInterpolate(A=@n147.RGB, B=@n148.RGB, Alpha=@n149)
    B=LinearInterpolate(A=@n150.RGB, B=@n151.RGB, Alpha=@n152)
    Alpha=NamedRerouteUsage()
   )
   False=@n147.RGB
  ) {parameterName=UseColorTintQuadrants}
  n154 = Constant()
  n155 = VectorParameter() {group=Color_MatLayer2, parameterName=Color_MatLayer2_A, defaultValue=<rgba 1 1 1 1>}
  n156 = VectorParameter() {group=Color_MatLayer2, parameterName=Color_MatLayer2_B, defaultValue=<rgba 1 1 1 1>}
  n157 = NamedRerouteUsage()
  n158 = VectorParameter() {group=Color_MatLayer2, parameterName=Color_MatLayer2_C, defaultValue=<rgba 1 1 1 1>}
  n159 = VectorParameter() {group=Color_MatLayer2, parameterName=Color_MatLayer2_D, defaultValue=<rgba 1 1 1 1>}
  n160 = NamedRerouteUsage()
  n161 = StaticSwitchParameter(
   True=LinearInterpolate(
    A=LinearInterpolate(A=@n155.RGB, B=@n156.RGB, Alpha=@n157)
    B=LinearInterpolate(A=@n158.RGB, B=@n159.RGB, Alpha=@n160)
    Alpha=NamedRerouteUsage()
   )
   False=@n155.RGB
  ) {parameterName=UseColorTintQuadrants}
  n162 = Constant4Vector()
  n163 = ComponentMask(None=MaterialFunctionCall(1=AppendVector(A=@n153, B=@n154), 2=AppendVector(A=@n161, B=@n154), 3=@n162.RGBA, 4=@n162.RGBA) {materialFunction=&MF_NeonCity_4Layer_Vec4}.Result) {b=true, g=true, r=true}
  NamedRerouteDeclaration(None=StaticSwitchParameter(True=Multiply(A=@n146, B=@n163), False=@n146) {parameterName=UseColorTint}) {name=Diffuse_4Layer.RGB, nodeColor=<rgba 0 0.8594 1 1>}
  MaterialExpressionReroute_7 = Reroute(None=@n1)
  NamedRerouteDeclaration(None=@MaterialExpressionReroute_7.MA In) {name=Debug_MA_In, nodeColor=<rgba 1 0 0.2578 1>}
  n164 = TextureCoordinate()
  n165 = Constant() {r=0.5}
  n166 = Constant()
  n167 = Constant() {r=1}
  NamedRerouteDeclaration(None=If(A=ComponentMask(None=@n164) {r=true}, B=@n165, "A > B"=@n166, "A == B"=@n166, "A < B"=@n167)) {name=QuadrantMask1, nodeColor=<rgba 0 0.25 1 1>}
  NamedRerouteDeclaration(None=Divide(A=NamedRerouteUsage(), B=ScalarParameter() {defaultValue=350, group=Stain, parameterName=Stain_UVsDivisor}) {constB=500}) {name=StainUVs, nodeColor=<rgba 0.1484 0 1 1>}
  NamedRerouteDeclaration(None=If(A=ComponentMask(None=@n164) {g=true}, B=@n165, "A > B"=@n166, "A == B"=@n166, "A < B"=@n167)) {name=QuadrantMask2, nodeColor=<rgba 0.8047 0 1 1>}
  n168 = TextureSampleParameter2D(UVs=NamedRerouteUsage()) {
   group=Stain
   parameterName=StainMap
   samplerSource=SSM_Wrap_WorldGroupSettings
   samplerType=SAMPLERTYPE_Masks
   texture=&T_White
  }
  n169 = NamedRerouteDeclaration(None=@n168.RGBA) {name=StainMap.RGBA, nodeColor=<rgba 0 0.2734 1 1>}
  NamedRerouteDeclaration(None=ChannelMaskParameter(None=@n169) {group=Ambient Occlusion, parameterName=Stain_AO_Channel}.R) {name=Stain AO, nodeColor=<rgba 1 0 0.2344 1>}
  n170 = ScalarParameter() {defaultValue=1, group=Color_MatLayer1, parameterName=Roughness_MatLayer1_A}
  n171 = LinearInterpolate(A=@n170, B=ScalarParameter() {defaultValue=1, group=Color_MatLayer1, parameterName=Roughness_MatLayer1_B}, Alpha=NamedRerouteUsage())
  n172 = LinearInterpolate(
   A=ScalarParameter() {defaultValue=1, group=Color_MatLayer1, parameterName=Roughness_MatLayer1_C}
   B=ScalarParameter() {defaultValue=1, group=Color_MatLayer1, parameterName=Roughness_MatLayer1_D}
   Alpha=NamedRerouteUsage()
  )
  n173 = NamedRerouteUsage()
  n174 = Constant3Vector()
  n175 = AppendVector(A=StaticSwitchParameter(True=LinearInterpolate(A=@n171, B=@n172, Alpha=@n173), False=@n170) {parameterName=UseRoughnessByQuadrants}, B=@n174)
  n176 = ScalarParameter() {defaultValue=1, group=Color_MatLayer2, parameterName=Roughness_MatLayer2_A}
  n177 = LinearInterpolate(A=@n176, B=ScalarParameter() {defaultValue=1, group=Color_MatLayer2, parameterName=Roughness_MatLayer2_B}, Alpha=NamedRerouteUsage())
  n178 = LinearInterpolate(
   A=ScalarParameter() {defaultValue=1, group=Color_MatLayer2, parameterName=Roughness_MatLayer2_C}
   B=ScalarParameter() {defaultValue=1, group=Color_MatLayer2, parameterName=Roughness_MatLayer2_D}
   Alpha=NamedRerouteUsage()
  )
  n179 = NamedRerouteUsage()
  n180 = AppendVector(A=StaticSwitchParameter(True=LinearInterpolate(A=@n177, B=@n178, Alpha=@n179), False=@n176) {parameterName=UseRoughnessByQuadrants}, B=@n174)
  n181 = Constant4Vector()
  NamedRerouteDeclaration(None=ComponentMask(None=MaterialFunctionCall(1=@n175, 2=@n180, 3=@n181.RGBA, 4=@n181.RGBA) {materialFunction=&MF_NeonCity_4Layer_Vec4}.Result) {r=true}) {name=RoughnessByQuad, nodeColor=<rgba 0 1 0.5313 1>}
  NamedRerouteUsage()
  n182 = ChannelMaskParameter(None=@n169) {group=Stain, maskChannel=Green, parameterName=Stain_Layer1_Channel, defaultValue=<rgba 0 1 0 0>}
  n183 = ScalarParameter() {defaultValue=1, group=Stain, parameterName=Stain_Layer1_Strength}
  n184 = ChannelMaskParameter(None=@n169) {group=Stain, maskChannel=Blue, parameterName=Stain_Layer2_Channel, defaultValue=<rgba 0 0 1 0>}
  n185 = ScalarParameter() {defaultValue=1, group=Stain, parameterName=Stain_Layer2_Strength}
  OneMinus(None=Multiply(A=LinearInterpolate(B=@n182.R, Alpha=@n183) {constA=1}, B=LinearInterpolate(B=@n184.R, Alpha=@n185) {constA=1}))
  PinBase() {reroutePins [MA In {expression=&self:MaterialExpressionReroute_7}]}
  n186 = GetMaterialAttributes(None=NamedRerouteUsage()) {attributeGetTypes [#1]}
  n187 = ChannelMaskParameter(None=NamedRerouteUsage()) {group=Debug, maskChannel=Blue, parameterName=PreviewWindowRandomMaskChannel, defaultValue=<rgba 0 0 1 0>}
  n188 = Multiply(A=@n187.R, B=ScalarParameter() {defaultValue=1, group=Debug, parameterName=PreviewWindowRandomMaskOpacity}) {constB=0.5}
  n189 = StaticSwitchParameter(
   True=SetMaterialAttributes!(MaterialAttributes, Base Color, Emissive Color=NamedRerouteUsage()) {attributeSetTypes [#1, #5]}
   False=StaticSwitchParameter(
    True=SetMaterialAttributes(Base Color=@n186.Base Color, Emissive Color=@n188) {attributeSetTypes [#1, #5]}
    False=NamedRerouteUsage()
   ) {group=Debug, parameterName=PreviewWindowRandomMask}
  ) {group=Debug, parameterName=PreviewWindowEmissive}
  PinBase!(MA Out=@n189) {reroutePins [MA Out {expression=&self:MaterialExpressionReroute_8}]}
  MaterialExpressionReroute_8 = Reroute(None=@n189)
  n190 = NamedRerouteUsage()
  n191 = ScalarParameter() {defaultValue=0.1, group="1) Base Layer", parameterName=BaseLayer_Specular_Bias}
  n192 = StaticSwitchParameter(
   True=LinearInterpolate(A=@n191, B=ScalarParameter() {defaultValue=0.1, group="2) Top Layer", parameterName=TopLayer_Specular_Bias}, Alpha=NamedRerouteUsage())
   False=@n191
  ) {defaultValue=true, group="2) Top Layer", parameterName=* UseTopLayer}
  StaticSwitchParameter(True=Add(A=Multiply(A=@n190, B=@n192), B=@n31), False=@n31) {defaultValue=true, group="1) Base Layer", parameterName=* UseBaseLayer}
  n193 = ComponentMask(None=@n40) {b=true, g=true, r=true}
  n194 = TextureSampleParameter2D(UVs=NamedRerouteUsage()) {
   group="1) Base Layer"
   parameterName=BaseLayerNormals
   samplerSource=SSM_Wrap_WorldGroupSettings
   samplerType=SAMPLERTYPE_Normal
   texture=&FlatNormalBlue
  }
  n195 = ScalarParameter() {defaultValue=1, group="1) Base Layer", parameterName=BaseLayer_NormalStrength, sliderMin=0.3}
  MaterialExpressionAppend3Vector_1 = MaterialXAppend3Vector(A=@n195, B=@n195, C=Constant())
  n196 = Multiply(A=@n194.RGB, B=@MaterialExpressionAppend3Vector_1)
  n197 = TextureSampleParameter2D(UVs=NamedRerouteUsage()) {
   group="2) Top Layer"
   parameterName=TopLayerNormals
   samplerSource=SSM_Wrap_WorldGroupSettings
   samplerType=SAMPLERTYPE_Normal
   texture=&FlatNormalBlue
  }
  n198 = ScalarParameter() {defaultValue=1, group="2) Top Layer", parameterName=TopLayer_NormalStrength, sliderMin=0.3}
  MaterialExpressionAppend3Vector_0 = MaterialXAppend3Vector(A=@n198, B=@n198, C=Constant())
  n199 = Multiply(A=@n197.RGB, B=@MaterialExpressionAppend3Vector_0)
  n200 = NamedRerouteUsage()
  n201 = Add(
   A=StaticSwitchParameter(True=LinearInterpolate(A=@n196, B=@n199, Alpha=@n200), False=@n196) {defaultValue=true, group="2) Top Layer", parameterName=* UseTopLayer}
   B=@n193
  )
  n202 = NamedRerouteUsage()
  StaticSwitchParameter(True=Normalize(VectorInput=LinearInterpolate(A=@n193, B=@n201, Alpha=@n202)), False=@n202) {defaultValue=true, group="1) Base Layer", parameterName=* UseBaseLayer}
  VertexInterpolator(VS=@n16.Result) {desc=unplugged to support dirt waterline}
 ]
}
